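# DudeBro_Amb_Growth_Cue is a looping ambient SoundCue for the CubeIsland event, reusing the Cube close-ambience waves under the Fort_Events sound class.
# A quality-level node at the root picks between two looping wave players.
# The first two quality slots play Cube_AmbientClose_01, and the last two play its _opt variant.
# The cue has no randomization or modulation nodes.
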SoundCue /Figment_S06/Athena/Sounds/Events/CubeIsland/DudeBro_Amb_Growth_Cue
root=QualityLevel_0
nodes [
 {id=QualityLevel_0, children [WavePlayer_4,WavePlayer_4,WavePlayer_5,WavePlayer_5]}
 {id=WavePlayer_4, properties {bLooping=true, soundWaveAssetPtr=/Game/Athena/Sounds/Events/Cube/Cube_AmbientClose_01}}
 {id=WavePlayer_5, properties {bLooping=true, soundWaveAssetPtr=/Game/Athena/Sounds/Events/Cube/Cube_AmbientClose_01_opt}}
]
properties {maxDistance=2097152, soundClassObject=&Fort_Events, volumeMultiplier=1}
