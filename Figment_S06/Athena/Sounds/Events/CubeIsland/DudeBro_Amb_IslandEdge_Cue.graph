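# Sound cue for looping ambience at the Cube Island edge, playing in the Fort_Events sound class.
# A looping distant-ambience wave runs through an overriding Attenuation node with logarithmic falloff, wide stereo spread and a custom curve.
# A Mixer boosts that branch and feeds the first two Quality Level slots, leaving the other two empty.
# The cue raises pitch and slightly lowers volume.
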SoundCue /Figment_S06/Athena/Sounds/Events/CubeIsland/DudeBro_Amb_IslandEdge_Cue
root=QualityLevel_0
nodes [
 {
  id=Attenuation_0
  children [WavePlayer_3]
  properties {
   bOverrideAttenuation=true
   attenuationOverrides {
    distanceAlgorithm=Logarithmic
    falloffDistance=4000
    stereoSpread=1024
    attenuationShapeExtents {x=0}
    customAttenuationCurve {editorCurveData {keys [<at 0 1>,<at 0.5 0>]}}
   }
  }
 }
 {id=Mixer_0, children [Attenuation_0], properties {inputVolume [2.5]}}
 {id=QualityLevel_0, children [Mixer_0,Mixer_0,null,null]}
 {id=WavePlayer_3, properties {bLooping=true, soundWaveAssetPtr=&/DudeBro_Amb_Nrml_Dist_01}}
]
properties {
 maxDistance=4000
 pitchMultiplier=1.5
 soundClassObject=&Fort_Events
 volumeMultiplier=0.8
 attenuationOverrides {bSpatialize=false, falloffDistance=5000, nonSpatializedRadiusStart=500, stereoSpread=5000, attenuationShapeExtents {x=0}}
 concurrencyOverrides {maxCount=1, resolutionRule=PreventNew}
}
