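# Looping in-world SoundCue for the DudeBro vent in the CubeIsland event, played on the events sound class with ForgedConsumable in-world attenuation.
# A QualityLevel node maps its first two quality slots to the standard looping wave and its last two to the "_opt" variant.
# An Enveloper above it loops indefinitely, modulating pitch over a twenty-second curve while holding volume flat.
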SoundCue /Figment_S06/Athena/Sounds/Events/CubeIsland/DudeBro_Vent_InWorldLoop_Cue
root=Enveloper_0
nodes [
 {
  id=Enveloper_0
  children [QualityLevel_0]
  properties {
   bLoop=true
   bLoopIndefinitely=true
   pitchCurve {editorCurveData {keys [{time=0}, <at 10 0.75>, {time=20}]}}
   volumeCurve {editorCurveData {keys [<at 0 1>]}}
  }
 }
 {id=QualityLevel_0, children [WavePlayer_0,WavePlayer_0,WavePlayer_1,WavePlayer_1]}
 {id=WavePlayer_0, properties {bLooping=true, soundWaveAssetPtr=/Game/Athena/Sounds/Events/CubeIsland/DudeBro_Vent_Loop_01}}
 {id=WavePlayer_1, properties {bLooping=true, soundWaveAssetPtr=/Game/Athena/Sounds/Events/CubeIsland/DudeBro_Vent_Loop_01_opt}}
]
properties {
 attenuationSettings=/Game/Sounds/Attenuation/Fort_ForgedConsumable_InWorld_Att
 maxDistance=1000
 soundClassObject=&Fort_Events
 volumeMultiplier=1.5
}
